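# Physics asset for the EpicControlRig quadruped wolf skeletal mesh, defining its simulated collision bodies and joints.
# Single capsules cover spine, neck, head, jaw, ears, paired front and hind legs, and the tail.
# Constraints form a tree from the pelvis through spine to chest, with head, limbs and tail branching off, bridging bones that have no body.
# Every joint uses the same uniform swing and twist limit angles rather than per-joint anatomical tuning.
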PhysicsAsset /EpicControlRig/Meshes/Quadruped/SKM_Quad_Wolf_PhysicsAsset
bodies {
 BipedHeadNeck_A_Head_C [{shapeType=Capsule, length=25.1182, radius=23.88, center=<15.0575 9.8485 0.0626>, rotation=<rot 0 116.0331 -89.8554>}]
 BipedHeadNeck_A_NeckBase_C [{shapeType=Capsule, length=35.1913, radius=24.3709, center=<6.0816 4.1941 0.2904>, rotation=<rot 0 176.1538 -88.2046>}]
 BipedHeadNeck_A_NeckMid_C [{shapeType=Capsule, length=30.3118, radius=24.7799, center=<5.3072 -1.4091 -0.1488>, rotation=<rot 0 172.1952 -88.3883>}]
 Leaf_Ear_Leaf_L [{shapeType=Capsule, length=15.7256, radius=10.5831, center=<7.4441 -6.896 -9.4249>, rotation=<rot 28.7129 -90 -25.6652>}]
 Leaf_Ear_Leaf_R [{shapeType=Capsule, length=15.7079, radius=10.8323, center=<-7.6199 6.8949 9.2888>, rotation=<rot 28.3117 -90 -25.2552>}]
 Leaf_Jaw_Leaf_C [{shapeType=Capsule, length=13.9486, radius=9.1648, center=<14.1998 0 0.7408>, rotation=<rot 0 90 -67.3712>}]
 PawedArm_A_Ball_L [{shapeType=Capsule, length=11.0608, radius=7.8085, center=<8.4834 2.3132 -0.8305>, rotation=<rot 28.4716 -90 40.9373>}]
 PawedArm_A_Ball_R [{shapeType=Capsule, length=11.0608, radius=7.8085, center=<-8.4834 -2.3132 0.8305>, rotation=<rot 28.4716 -90 40.9373>}]
 PawedArm_A_Elbow_L [{shapeType=Capsule, length=20.5304, radius=8.9417, center=<16.4713 -0.0815 -0.0669>, rotation=<rot 0 79.4811 -85.9596>}]
 PawedArm_A_Elbow_R [{shapeType=Capsule, length=20.5304, radius=8.9417, center=<-16.4713 0.0815 0.0669>, rotation=<rot 0 79.4811 -85.9596>}]
 PawedArm_A_Shoulder_L [{shapeType=Capsule, length=22.1016, radius=15.2066, center=<12.8351 0.931 -2.4559>, rotation=<rot 0 123.0226 -86.4969>}]
 PawedArm_A_Shoulder_R [{shapeType=Capsule, length=21.9567, radius=15.24, center=<-13.2155 -0.7575 2.4437>, rotation=<rot 0 121.5863 -86.4184>}]
 PawedLeg_A_Ankle_L [{shapeType=Capsule, length=17.9492, radius=5.7371, center=<13.2724 -1.1081 -0.016>, rotation=<rot 0 -81.7093 -88.7648>}]
 PawedLeg_A_Ankle_R [{shapeType=Capsule, length=17.9492, radius=5.7371, center=<-13.2724 1.1081 0.016>, rotation=<rot 0 -81.7093 -88.7648>}]
 PawedLeg_A_Thigh_L [{shapeType=Capsule, length=24.585, radius=16.6585, center=<19.3541 -3.761 1.2456>, rotation=<rot 0 37.7321 -85.7811>}]
 PawedLeg_A_Thigh_R [{shapeType=Capsule, length=24.6045, radius=16.5898, center=<-19.5222 3.8352 0.9997>, rotation=<rot 0 38.294 -85.4362>}]
 PawedLeg_A_Toe_L [{shapeType=Capsule, length=9.4312, radius=7.7907, center=<7.2048 0.8913 0.2497>, rotation=<rot -3.7988 -90 17.8712>}]
 PawedLeg_A_Toe_R [{shapeType=Capsule, length=9.4312, radius=7.7907, center=<-7.2048 -0.8913 -0.2497>, rotation=<rot -3.7987 -90 17.8712>}]
 QuadSpine_A_Chest_C [{shapeType=Capsule, length=34.8544, radius=31.9141, center=<3.1648 1.4446 0.156>, rotation=<rot 0 -10.6793 -85.6398>}]
 QuadSpine_A_Pelvis_C [{shapeType=Capsule, length=15.4501, radius=13.4391, center=<5.3105 -1.4227 -0.0042>, rotation=<rot 0 164.3129 -89.5733>}]
 QuadSpine_A_Spine2_C [{shapeType=Capsule, length=20.7369, radius=20.2231, center=<7.0127 -0.9915 0.8272>, rotation=<rot 0 -175.9402 -174.7134>}]
 QuadSpine_A_Spine3_C [{shapeType=Capsule, length=26.1647, radius=8.8176, center=<2.6542 0.8676 0.103>, rotation=<rot 0 -171.2785 -89.2205>}]
 Tail_A_Tail1_C [{shapeType=Capsule, length=14.8632, radius=13.5125, center=<16.4516 0.5382 -2.2141>, rotation=<rot 0 87.6922 -74.0261>}]
 Tail_A_Tail3_C [{shapeType=Capsule, length=17.6791, radius=16.5521, center=<6.5042 0.0149 1.2852>, rotation=<rot -0.1934 -90 7.8091>}]
 Tail_A_Tail4_C [{shapeType=Capsule, length=16.9166, radius=16.6141, center=<8.8284 -0.0633 -2.0824>, rotation=<rot 0 0.2207 -60.0812>}]
 Tail_A_Tail5_C [{shapeType=Capsule, length=18.211, radius=14.9742, center=<12.0031 0.7368 2.1723>, rotation=<rot 0 95.6852 -76.3292>}]
 Tail_A_TailBase_C [{shapeType=Capsule, length=13.8069, radius=11.8808, center=<12.1464 0.4435 4.8254>, rotation=<rot 0 -1.6823 -87.1467>}]
}
constraints [
 {bone1Name=QuadSpine_A_Spine2_C, bone2Name=QuadSpine_A_Pelvis_C, swing1LimitDegrees=45, swing2LimitDegrees=45, twistLimitDegrees=45}
 {bone1Name=QuadSpine_A_Spine3_C, bone2Name=QuadSpine_A_Spine2_C, swing1LimitDegrees=45, swing2LimitDegrees=45, twistLimitDegrees=45}
 {bone1Name=QuadSpine_A_Chest_C, bone2Name=QuadSpine_A_Spine3_C, swing1LimitDegrees=45, swing2LimitDegrees=45, twistLimitDegrees=45}
 {bone1Name=PawedArm_A_Shoulder_L, bone2Name=QuadSpine_A_Chest_C, swing1LimitDegrees=45, swing2LimitDegrees=45, twistLimitDegrees=45}
 {bone1Name=PawedArm_A_Elbow_L, bone2Name=PawedArm_A_Shoulder_L, swing1LimitDegrees=45, swing2LimitDegrees=45, twistLimitDegrees=45}
 {bone1Name=PawedArm_A_Ball_L, bone2Name=PawedArm_A_Elbow_L, swing1LimitDegrees=45, swing2LimitDegrees=45, twistLimitDegrees=45}
 {bone1Name=BipedHeadNeck_A_NeckBase_C, bone2Name=QuadSpine_A_Chest_C, swing1LimitDegrees=45, swing2LimitDegrees=45, twistLimitDegrees=45}
 {
  bone1Name=BipedHeadNeck_A_NeckMid_C
  bone2Name=BipedHeadNeck_A_NeckBase_C
  swing1LimitDegrees=45
  swing2LimitDegrees=45
  twistLimitDegrees=45
 }
 {bone1Name=BipedHeadNeck_A_Head_C, bone2Name=BipedHeadNeck_A_NeckMid_C, swing1LimitDegrees=45, swing2LimitDegrees=45, twistLimitDegrees=45}
 {bone1Name=Leaf_Jaw_Leaf_C, bone2Name=BipedHeadNeck_A_Head_C, swing1LimitDegrees=45, swing2LimitDegrees=45, twistLimitDegrees=45}
 {bone1Name=Leaf_Ear_Leaf_L, bone2Name=BipedHeadNeck_A_Head_C, swing1LimitDegrees=45, swing2LimitDegrees=45, twistLimitDegrees=45}
 {bone1Name=Leaf_Ear_Leaf_R, bone2Name=BipedHeadNeck_A_Head_C, swing1LimitDegrees=45, swing2LimitDegrees=45, twistLimitDegrees=45}
 {bone1Name=PawedArm_A_Shoulder_R, bone2Name=QuadSpine_A_Chest_C, swing1LimitDegrees=45, swing2LimitDegrees=45, twistLimitDegrees=45}
 {bone1Name=PawedArm_A_Elbow_R, bone2Name=PawedArm_A_Shoulder_R, swing1LimitDegrees=45, swing2LimitDegrees=45, twistLimitDegrees=45}
 {bone1Name=PawedArm_A_Ball_R, bone2Name=PawedArm_A_Elbow_R, swing1LimitDegrees=45, swing2LimitDegrees=45, twistLimitDegrees=45}
 {bone1Name=PawedLeg_A_Thigh_L, bone2Name=QuadSpine_A_Pelvis_C, swing1LimitDegrees=45, swing2LimitDegrees=45, twistLimitDegrees=45}
 {bone1Name=PawedLeg_A_Ankle_L, bone2Name=PawedLeg_A_Thigh_L, swing1LimitDegrees=45, swing2LimitDegrees=45, twistLimitDegrees=45}
 {bone1Name=PawedLeg_A_Toe_L, bone2Name=PawedLeg_A_Ankle_L, swing1LimitDegrees=45, swing2LimitDegrees=45, twistLimitDegrees=45}
 {bone1Name=Tail_A_TailBase_C, bone2Name=QuadSpine_A_Pelvis_C, swing1LimitDegrees=45, swing2LimitDegrees=45, twistLimitDegrees=45}
 {bone1Name=Tail_A_Tail1_C, bone2Name=Tail_A_TailBase_C, swing1LimitDegrees=45, swing2LimitDegrees=45, twistLimitDegrees=45}
 {bone1Name=Tail_A_Tail3_C, bone2Name=Tail_A_Tail1_C, swing1LimitDegrees=45, swing2LimitDegrees=45, twistLimitDegrees=45}
 {bone1Name=Tail_A_Tail4_C, bone2Name=Tail_A_Tail3_C, swing1LimitDegrees=45, swing2LimitDegrees=45, twistLimitDegrees=45}
 {bone1Name=Tail_A_Tail5_C, bone2Name=Tail_A_Tail4_C, swing1LimitDegrees=45, swing2LimitDegrees=45, twistLimitDegrees=45}
 {bone1Name=PawedLeg_A_Thigh_R, bone2Name=QuadSpine_A_Pelvis_C, swing1LimitDegrees=45, swing2LimitDegrees=45, twistLimitDegrees=45}
 {bone1Name=PawedLeg_A_Ankle_R, bone2Name=PawedLeg_A_Thigh_R, swing1LimitDegrees=45, swing2LimitDegrees=45, twistLimitDegrees=45}
 {bone1Name=PawedLeg_A_Toe_R, bone2Name=PawedLeg_A_Ankle_R, swing1LimitDegrees=45, swing2LimitDegrees=45, twistLimitDegrees=45}
]
